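# Translucent, unlit, two-sided material for Zero Point aurora spline glows, judging by its name and river vertical glow folder.
# Emissive layers panning star, noise-warped and caustic textures over a top-to-bottom color gradient.
# Opacity combines noise, edge masks and distance fades with collection parameters, replacing angle, depth and near fades with a constant on forward and mobile.
# World position offset uses vertex green (main branch) and blue (branches) for height stretch and sine waves, plus camera offset and distance-scaling masks.
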Material /Game/Environments/Hera/FX/ZeroPoint/RiverVerticleGlow/M_ZeroPoint_AuroraSplines
graph {
 properties {
  bAutomaticallySetUsageInEditor=false
  blendMode=BLEND_Translucent
  previewMesh=/Game/Effects/Fort_Effects/Effects/Environment/Hera/Shards/Glow/SM_Env_TeaCake_Shard_Glows
  shadingModel=MSM_Unlit
  twoSided=true
  usedShadingModels=Unlit
 }
 nodes [
  n1 = ComponentMask(None=TextureCoordinate()) {g=true}
  n2 = ScalarParameter() {defaultValue=2, parameterName=EmissiveGradientPower}
  n3 = Constant() {r=0.25}
  n4 = TextureCoordinate()
  n5 = Constant2Vector() {g=1, r=1}
  n6 = TextureCoordinate()
  n7 = LinearInterpolate(
   A=Multiply(
    A=MaterialFunctionCall!(RotationAmount=@n3, Coords=@n4, Pivot) {materialFunction=/BaseMaterial/Materials/Functions/MF_Rotate2D}.Result
    B=Constant2Vector() {g=1, r=1}
   )
   B=TextureSample(UVs=Multiply(A=@n5, B=@n6)) {samplerType=SAMPLERTYPE_LinearColor, texture=&HighContrastNoise}.G
  ) {constAlpha=0.05}
  n8 = Panner(Coordinate=@n7) {speedX=0.03, speedY=0.035}
  n9 = TextureSampleParameter2D(UVs=Panner(Coordinate=@n7) {speedX=0.04, speedY=0.025}) {parameterName=MainTexture, texture=/Game/Athena/Environments/Nexus/Textures/T_ZeroPoint_Line_Stars}
  n10 = VectorParameter() {parameterName=Tint, defaultValue=<rgba 0.6244 0 1 1>}
  n11 = TextureCoordinate() {vTiling=-1}
  n12 = Constant2Vector() {g=4, r=4}
  n13 = TextureSample(UVs=Panner(Coordinate=Multiply(A=@n11, B=@n12)) {speedY=0.015}) {texture=/Game/Effects/Fort_Effects/Textures/Tile/T_Caustics_Chromatic}
  n14 = Add(
   A=Add(
    A=TextureSampleParameter2D(UVs=@n8) {parameterName=MainTexture, texture=&T_ZeroPoint_Line_Stars}.RGB
    B=Multiply(A=@n9.RGB, B=@n10.RGB)
   )
   B=@n13.RGB
  )
  n15 = LinearInterpolate(
   A=VectorParameter() {parameterName=TopColor, defaultValue=<rgba 0.8338 0 0.9896 1>}.RGB
   B=VectorParameter() {parameterName=BottomColor, defaultValue=<rgba 0.899 3 2.2034 1>}.RGB
   Alpha=ComponentMask(None=TextureCoordinate()) {g=true}
  )
  n16 = Add(A=@n14, B=@n15)
  MaterialFunctionCall(Emissive=Multiply(A=Power(Base=@n1, Exp=@n2), B=Power(Base=@n16))) {materialFunction=&MF_FX_EmissiveAdjustments}
  n17 = Multiply(A=TextureSample(UVs=Panner() {speedX=-0.1}) {texture=/Game/Global/Textures/Default/Noise/T_SoftNoise}.R) {constB=4}
  n18 = TextureCoordinate() {vTiling=-1}
  n19 = Constant2Vector() {g=1, r=2}
  n20 = TextureObject() {samplerType=SAMPLERTYPE_LinearColor, texture=&HighContrastNoise}
  n21 = TextureSample(UVs=Panner(Coordinate=Multiply(A=@n18, B=@n19)) {speedY=0.05}, Tex=@n20) {samplerType=SAMPLERTYPE_LinearColor, texture=&HighContrastNoise}
  n22 = TextureSample(UVs=Panner() {speedX=-0.1}, Tex=@n20) {samplerType=SAMPLERTYPE_LinearColor, texture=&HighContrastNoise}
  n23 = MaterialFunctionCall!(
   UVs=TextureCoordinate()
   Mult=ScalarParameter() {defaultValue=5, parameterName=EdgeMult}
   Exp=ScalarParameter() {defaultValue=3, parameterName=EdgeExp}
  ) {materialFunction=/Game/Effects/Fort_Effects/MaterialFunctions/MF_SquareEdgeMask}
  n24 = Add(A=@n9.B, B=@n13.B)
  n25 = Multiply(A=@n23.SquareMask) {constB=0.1}
  n26 = MaterialFunctionCall(UVs=TextureCoordinate(), Mult=Constant() {r=2.5}, Exp=Constant() {r=4}) {materialFunction=&MF_SquareEdgeMask}
  n27 = Multiply(A=Power(Base=Add(A=@n21.R, B=@n22.R)), B=Add(A=Multiply(A=@n23.SquareMask, B=@n24), B=Add(A=@n25, B=@n26.SquareMask)))
  n28 = ScalarParameter() {defaultValue=1, parameterName=AngleFade}
  n29 = Saturate(None=Multiply(A=MaterialFunctionCall("FadeSensitivity (optional)"=@n28) {materialFunction=&FadeBasedOnViewAngle}.Result) {constB=2})
  n30 = ScalarParameter() {defaultValue=200, parameterName=DepthFade}
  n31 = ScalarParameter() {defaultValue=200, parameterName=DepthFade_Far}
  n32 = NamedRerouteUsage()
  n33 = Multiply(
   A=MaterialFunctionCall(Distance=ScalarParameter() {defaultValue=800, parameterName=NearFadeDIst}) {materialFunction=&Near_Camera_Fade}.Result
   B=DepthFade(FadeDistance=LinearInterpolate(A=@n30, B=@n31, Alpha=@n32))
  )
  n34 = Constant() {r=0.7}
  n35 = Multiply(
   A=Multiply(A=Multiply(A=@n17, B=@n27), B=ScalarParameter() {defaultValue=0.8, parameterName=OverallOpacity})
   B=ShadingPathSwitch(Default=Multiply(A=@n29, B=@n33), Forward=@n34, Mobile=@n34)
  )
  n36 = Length(None=Subtract(A=CameraPositionWS(), B=WorldPosition().XYZ)) {is=class MaterialExpressionLength}
  n37 = ScalarParameter() {defaultValue=3000, parameterName=StartFade}
  n38 = ScalarParameter() {defaultValue=8000, parameterName=EndFade}
  n39 = Saturate(None=Divide(A=Subtract(A=@n36, B=@n37), B=Subtract(A=@n38, B=@n37)))
  n40 = ComponentMask(None=TextureCoordinate()) {r=true}
  n41 = CollectionParameter() {collection=&/MPC_ZP_TendrilOpacity, parameterName=UOpacity_Power}
  Multiply(
   A=Multiply(A=Multiply(A=@n35, B=@n39), B=Power(Base=@n40, Exp=@n41))
   B=CollectionParameter() {collection=&MPC_ZP_TendrilOpacity, parameterName=GlobalOpacity_Mult}
  )
  Multiply(A=Multiply(A=@n14, B=@n15), B=@n16) {constB=4}
  n42 = VertexColor()
  n43 = TextureCoordinate()
  n44 = LinearInterpolate(
   A=ScalarParameter() {defaultValue=5000, group=Main Branch, parameterName=HeightStretch}
   B=ScalarParameter() {defaultValue=15000, group=Main Branch, parameterName=HeightStretch_Max}
   Alpha=NamedRerouteUsage()
  )
  n45 = NamedRerouteUsage()
  n46 = Multiply(A=Multiply(A=ComponentMask(None=@n43) {g=true}, B=LinearInterpolate(B=@n44, Alpha=@n45)), B=Constant3Vector() {constant=<rgba 0 0 1 0>})
  n47 = VertexColor()
  n48 = ComponentMask(None=TextureCoordinate()) {g=true}
  n49 = NamedRerouteUsage()
  n50 = LinearInterpolate(
   B=LinearInterpolate(
    A=ScalarParameter() {defaultValue=5000, group=Branches, parameterName=BranchHeightStretch}
    B=ScalarParameter() {defaultValue=3000, group=Branches, parameterName=BranchHeightStretch_Max}
    Alpha=@n49
   )
   Alpha=NamedRerouteUsage()
  )
  n51 = Constant3Vector() {constant=<rgba 0 0 1 0>}
  n52 = ScalarParameter() {defaultValue=3000, parameterName=Param_3}
  n53 = Add(A=Multiply(A=Multiply(A=@n48, B=@n50), B=@n51), B=LinearInterpolate(B=Multiply(A=@n51, B=@n52), Alpha=@n49))
  n54 = Multiply(A=Time(), B=ScalarParameter() {defaultValue=0.0025, group=Branches, parameterName=Branches Sine Speed}) {constB=0.01}
  n55 = ComponentMask(None=TextureCoordinate()) {r=true}
  n56 = Sine(None=Multiply(A=Add(A=@n54, B=@n55), B=ScalarParameter() {defaultValue=22, group=Branches, parameterName=Branch Freq Strength}))
  n57 = ScalarParameter() {defaultValue=1250, group=Branches, parameterName=Branch Offset Strength}
  n58 = Multiply(A=Multiply(A=VertexColor().B, B=Multiply(A=@n56, B=@n57)), B=Constant3Vector() {constant=<rgba 0 0 1 0>})
  n59 = Multiply(A=Time(), B=ScalarParameter() {defaultValue=0.0025, group=Branches, parameterName=H Branches Sine Speed}) {constB=0.01}
  n60 = ComponentMask(None=TextureCoordinate()) {r=true}
  n61 = Sine(None=Multiply(A=Add(A=@n59, B=@n60), B=ScalarParameter() {defaultValue=22, group=Branches, parameterName=H Branch Freq Strength}))
  n62 = ScalarParameter() {defaultValue=1250, group=Branches, parameterName=H Branch Offset Strength}
  n63 = Multiply(A=Multiply(A=VertexColor().B, B=Multiply(A=@n61, B=@n62)), B=Constant3Vector() {constant=<rgba 0 0 1 0>})
  n64 = Multiply(A=Time(), B=ScalarParameter() {defaultValue=0.0025, group=Main Branch, parameterName=Sine Speed}) {constB=0.01}
  n65 = ComponentMask(None=TextureCoordinate()) {r=true}
  n66 = Sine(None=Multiply(A=Add(A=@n64, B=@n65), B=ScalarParameter() {defaultValue=22, group=Main Branch, parameterName=Freq Strength}))
  n67 = ScalarParameter() {defaultValue=1250, group=Main Branch, parameterName=Offset Strength}
  n68 = Multiply(A=Multiply(A=VertexColor().G, B=Multiply(A=@n66, B=@n67)), B=Constant3Vector() {constant=<rgba 0 0 1 0>})
  n69 = Multiply(A=Time(), B=ScalarParameter() {defaultValue=0.0025, group=Main Branch, parameterName=Horizontal Sine Speed}) {constB=0.01}
  n70 = ComponentMask(None=TextureCoordinate()) {r=true}
  n71 = Sine(None=Multiply(A=Add(A=@n69, B=@n70), B=ScalarParameter() {defaultValue=22, group=Main Branch, parameterName=Horizontal Freq Strength}))
  n72 = ScalarParameter() {defaultValue=1250, group=Main Branch, parameterName=Horizontal Offset Strength}
  n73 = Multiply(A=Multiply(A=VertexColor().G, B=Multiply(A=@n71, B=@n72)), B=Constant3Vector() {constant=<rgba 1 0 0 0>})
  n74 = Add(A=Add(A=Multiply(A=@n42.G, B=@n46), B=Multiply(A=@n47.B, B=@n53)), B=Add(A=Add(A=@n58, B=@n63), B=Add(A=@n68, B=@n73)))
  n75 = MaterialFunctionCall(Offset Amount=ScalarParameter() {defaultValue=4000, parameterName=Param_2}) {materialFunction=/Engine/Functions/Engine_MaterialFunctions02/WorldPositionOffset/CameraOffset}
  NamedRerouteDeclaration(None=LinearInterpolate(A=@n74, B=Add(A=@n74, B=@n75.Camera Offset), Alpha=NamedRerouteUsage())) {name=WPO, nodeColor=<rgba 0.6641 1 0 1>}
  NamedRerouteDeclaration(None=MaterialFunctionCall(
    Fade_Start_Distance=ScalarParameter() {defaultValue=1000, group=Scale, parameterName=HeightStretch_DistanceFade_EndFading}
    Fade_End_Distance=ScalarParameter() {defaultValue=30000, group=Scale, parameterName=HeightStretch_DistanceFade_StartFading}
   ) {materialFunction=&MF_Distance_SyncedFade_FX}.Result) {name=DistanceValueScale, nodeColor=<rgba 1 0.1641 0 1>}
  n76 = Length(None=Subtract(A=CameraPositionWS(), B=WorldPosition().XYZ))
  n77 = Constant()
  n78 = ScalarParameter() {defaultValue=25000, parameterName=WPODistanceScaling_Start}
  n79 = Constant() {r=0.2}
  NamedRerouteDeclaration(None=OneMinus(None=SphereMask(A=@n76, B=@n77, Radius=@n78, Hardness=@n79) {hardnessPercent=0.2})) {name=WPODistanceScaling, nodeColor=<rgba 0.5938 0 1 1>}
  NamedRerouteUsage()
  ScalarParameter() {parameterName=U_Opacity_Power}
  VertexColor()
 ]
}
